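# Niagara system for the DuneBriefShift pickaxe's equip and idle effects, with toon and non-toon looks selected by System.IsToon.
# Idle pins three partitioned particles to skeletal mesh bones or sockets, using per-partition tags to drive a ribbon, pulsing energy ball and blade-blur meshes.
# Equip spawns short-lived mesh bits around animated radial offsets, while EquipTracer adds a fading mesh with animated material parameters.
# Equip and blade-blur colors follow time-of-day curves.
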NiagaraSystem /BRCosmetics/Weapons/FORT_Melee/Pickaxe_DuneBriefShift/FX/NS_Pickaxe_DuneBriefShift_Idle_Universal
dependencies {
 usedDataInterfaces [
  &NiagaraDataInterfaceArrayFloat3
  &NiagaraDataInterfaceVectorCurve
  &NiagaraDataInterfaceColorCurve
  &NiagaraDataInterfaceArrayFloat
  &NiagaraDataInterfaceArrayInt32
  &NiagaraDataInterfaceArrayFloat2
  &NiagaraDataInterfaceCurve
  &NiagaraDataInterfaceSkeletalMesh
 ]
 usedDynamicInputs [&MergeHelper_ColorSwitch, &MergeHelper_FloatSwitch]
 usedModules [
  &self:SystemSpawnScript.SystemScriptSource.SystemScriptGraph.NiagaraNodeAssignment_0.SetVariables_#1
  &SystemState
  &self:SystemSpawnScript.SystemScriptSource.SystemScriptGraph.NiagaraNodeAssignment_1.SetVariables_#2
 ]
}
emitters {
 Equip {
  emitterUpdateScript [
   &EmitterState(Inactive Response=::NewEnumerator1, Life Cycle Mode=::NewEnumerator1, Loop Behavior=::NewEnumerator1, Loop Duration=0.65, UseLoopDelay=false)
   /Game/Effects/Niagara/Modules/Spawn/FN_Spawn_TriggeredBurst(Cooldown=0.01, Module.Use Burst Count Limit=false, Reset Mode=2, Spawn Count=20, UseLoopCountLimit=false) {enabled=false}
   &SpawnBurst_Instantaneous(Spawn Count=30, Spawn Time=0.25, Use Spawn Probability=false, UseLoopCountLimit=false)
   &SpawnRate(
    Spawn Probability=&MakeCustomFloatFromBool(
     Bool=&SetBoolByFloatComparison(A=$Emitter.NormalizedLoopAge, B=0.5, Comparison Type=::NewEnumerator6)
     True Float=1
     False Float=0
    )
    SpawnRate=200
    Use Spawn Probability=true
   )
  ]
  particleSpawnScript [
   &InitializeParticle(
    Color=&RandomRangeLinearColor(Minimum=<rgba 0.0107 0.076 0.2396 1>, Maximum=<rgba 0.0207 0.1158 0.4635 1>)
    Lifetime=&RandomRangeFloat(Minimum=0.15, Maximum=0.35)
    Mesh Scale Mode=::NewEnumerator3
    Mesh Uniform Scale=&RandomRangeFloat(Minimum=0.2, Maximum=0.75)
    Sprite Size Mode=::NewEnumerator3
    Uniform Sprite Size=&RandomRangeFloat(Minimum=10, Maximum=20)
    UsePositionOffset=false
    Write Material Random=true
   )
   /SetVariables(Particles.ParticleRandom:NiagaraFloat=&RandomRangeFloat(Minimum=0, Maximum=1)) {isSetParameters=true}
   /Game/Effects/Niagara/Modules/Curves/FN_ListCurve(
    Apply Position=true
    Curve Index:NiagaraFloat=&FloatFromCurve(
     FloatCurve=DataInterface() {propertyValues {Curve {keys [
         {arriveTangent=0.5817, interpMode=RCIM_Cubic, leaveTangent=0.5817, leaveTangentWeight=0.1157, tangentMode=RCTM_User, time=0, value=1}
         {tangentMode=RCTM_User, time=0.3, value=0}
         <at 0.35 0.121>
        ]}}}
     CurveIndex=$Emitter.LoopedAge
    )
    Input Type=2
    Number Inputs Switch=0
    Position Coordinate Space=1
    Vector_01:Vector3f=&RandomRangeVector(Minimum=<0 0 20>, Maximum=<0 0 28>)
    Vector_02:Vector3f=<0 0 -10>
   )
   &Color(Color=&MultiplyLinearColor(
     Linear Color=$Particles.Color
     Scale Factor=&MakeVector4_FromLinearColor(LinearColor=&MakeCustomLinearColorFromBool(
       Bool=$System.IsToon
       True Color=&ColorFromCurve(
        ColorCurve=DataInterface() {propertyValues {
          AlphaCurve {keys [<at 0 1 RCIM_Cubic>,<at 1 0 RCIM_Cubic>]}
          BlueCurve {keys [<at 0 0.7529 RCIM_Cubic>,<at 0.135 0.9897>,<at 0.8905 0.9987>,<at 1 0.7529 RCIM_Cubic>]}
          GreenCurve {keys [<at 0 0.7529 RCIM_Cubic>,<at 0.135 0.9862>,<at 0.8905 0.9983>,<at 1 0.7529 RCIM_Cubic>]}
          RedCurve {keys [<at 0 0.7529 RCIM_Cubic>,<at 0.135 0.9842>,<at 0.8905 0.998>,<at 1 0.7529 RCIM_Cubic>]}
         }}
        CurveIndex=Linked() {linkedVariable {name=NPC.FortniteNPC.TimeOfDay_0-1, type=NiagaraFloat}}
       )
       False Color=&ColorFromCurve(
        ColorCurve=DataInterface() {propertyValues {
          AlphaCurve {keys [<at 0 1 RCIM_Cubic>,<at 1 0 RCIM_Cubic>]}
          BlueCurve {keys [<at 0 0.4 RCIM_Cubic>,<at 0.135 0.9897>,<at 0.8905 0.9987>,<at 1 0.4 RCIM_Cubic>]}
          GreenCurve {keys [<at 0 0.4 RCIM_Cubic>,<at 0.135 0.9862>,<at 0.8905 0.9983>,<at 1 0.4 RCIM_Cubic>]}
          RedCurve {keys [<at 0 0.4 RCIM_Cubic>,<at 0.135 0.9842>,<at 0.8905 0.998>,<at 1 0.4 RCIM_Cubic>]}
         }}
        CurveIndex=Linked() {linkedVariable {name=NPC.FortniteNPC.TimeOfDay_0-1, type=NiagaraFloat}}
       )
      ))
    ))
  ]
  particleUpdateScript [
   &SolveForcesAndVelocity(Clamp Velocity=false, Limit Acceleration=false)
   &ParticleState()
   /SetVariables(
    Particles.Position=&self:ScratchDynamicInput(
     Theta:NiagaraFloat=$Particles.ParticleRandom
     Evaluation Type:NiagaraInt32=0
     Offset:Vector3f=Linked() {linkedVariable {name=Particles.Initial.Position, type=NiagaraPosition}}
     Rotation:Vector3f=<0 0 0>
     Radius:NiagaraFloat=&FloatFromCurve(
      FloatCurve=DataInterface() {propertyValues {Curve {keys [
          {arriveTangent=6.8794, interpMode=RCIM_Cubic, leaveTangent=6.8794, tangentMode=RCTM_User, time=0, value=1}
          {arriveTangent=0.8488, interpMode=RCIM_Cubic, leaveTangent=0.8488, tangentMode=RCTM_User, time=0.15, value=1.5}
          {arriveTangent=-1.1354, interpMode=RCIM_Cubic, leaveTangent=-1.1354, tangentMode=RCTM_User, time=0.3, value=1.5}
          {interpMode=RCIM_Cubic, tangentMode=RCTM_User, time=1, value=0.25}
         ]}}}
      Scale Curve=&FloatFromCurve(
       FloatCurve=DataInterface() {propertyValues {Curve {keys [{interpMode=RCIM_Cubic, tangentMode=RCTM_User, time=0, value=1}, {arriveTangent=-3.9037, tangentMode=RCTM_User, time=0.5, value=0}]}}}
       CurveIndex=$Emitter.LoopedAge
       Scale Curve=10
      )
     )
     Random Seed:NiagaraInt32=0
     U Coverage:NiagaraFloat=0
    )
    Particles.SpriteSize=&Multiply_Vector2D_ByFloat(
     Float=&FloatFromCurve(FloatCurve=DataInterface() {propertyValues {Curve {keys [
          {interpMode=RCIM_Cubic, tangentMode=RCTM_User, time=0, value=0}
          {interpMode=RCIM_Cubic, tangentMode=RCTM_User, time=0.25, value=1}
          {arriveTangent=-4.7366, interpMode=RCIM_Cubic, leaveTangent=-4.7366, tangentMode=RCTM_User, time=1, value=0}
         ]}}})
     Vector2D=$Particles.Initial.SpriteSize
    )
   ) {isSetParameters=true}
   &CameraOffset()
   &ScaleMeshSize(Scale Factor=&Multiply_VectorByFloat(
     Vector=<1 1 1>
     Float=&FloatFromCurve(FloatCurve=DataInterface() {propertyValues {Curve {keys [
          {arriveTangent=2.9773, interpMode=RCIM_Cubic, leaveTangent=2.9773, tangentMode=RCTM_User, time=0, value=1}
          {tangentMode=RCTM_User, time=1, value=0.25}
         ]}}})
    ))
   &ScaleColor(
    Scale RGB=&Multiply_VectorByFloat(
     Vector=<1 1 1>
     Float=&FloatFromCurve(FloatCurve=DataInterface() {propertyValues {Curve {keys [{arriveTangent=-29.5871, interpMode=RCIM_Cubic, leaveTangent=-29.5871, tangentMode=RCTM_User, time=0, value=10}, <at 0.5 1>]}}})
    )
    ScaleA=true
    ScaleRGB=true
   )
  ]
  rendererClasses [&NiagaraMeshRendererProperties]
  renderers [
   {rendererClass=&NiagaraMeshRendererProperties, rendererIndex=0}
   {rendererClass=&NiagaraMeshRendererProperties, rendererIndex=1}
  ]
 }
 EquipTracer {
  emitterUpdateScript [
   &EmitterState(Inactive Response=::NewEnumerator1, Life Cycle Mode=::NewEnumerator1, Loop Behavior=::NewEnumerator1, Loop Duration=0.85, UseLoopDelay=false)
   &SpawnBurst_Instantaneous(Use Spawn Probability=false, UseLoopCountLimit=false)
  ]
  particleSpawnScript [&InitializeParticle(Color=<rgba 15 15 15 1>, Lifetime=0.65, Mesh Scale Mode=::NewEnumerator3, UsePositionOffset=false, Write Material Random=true)]
  particleUpdateScript [
   &ParticleState()
   &DynamicMaterialParameters(
    Index 0 Param 1=1
    Index 0 Param 2=1
    Index 0 Param 4=1
    Index 1 Param 3=&FloatFromCurve(FloatCurve=DataInterface() {propertyValues {Curve {keys [
         {arriveTangent=-0.2519, interpMode=RCIM_Cubic, leaveTangent=-0.2519, tangentMode=RCTM_User, time=0, value=1}
         {arriveTangent=-1.804, tangentMode=RCTM_User, time=1, value=-1}
        ]}}})
    Module.Index0Float0Write=true
    Module.Index0Float1Write=true
    Module.Index0Float2Write=true
    Module.Index0Float3Write=true
    Module.Index1Float0Write=true
    Module.Index1Float1Write=true
    Module.Index1Float2Write=true
    Module.Index1Float3Write=true
    Write Parameter Index 1=::NewEnumerator1
   )
   &ScaleColor(
    Scale Alpha=&FloatFromCurve(
     FloatCurve=DataInterface() {propertyValues {Curve {keys [
         {interpMode=RCIM_Cubic, tangentMode=RCTM_User, time=0, value=1}
         {arriveTangent=-1, interpMode=RCIM_Cubic, leaveTangent=-1, time=0.4847, value=0.8588}
         {arriveTangent=-0.6755, tangentMode=RCTM_User, time=1, value=0}
        ]}}}
     CurveIndex=$Emitter.NormalizedLoopAge
    )
    ScaleA=true
    ScaleRGB=true
   )
   &CameraOffset(Camera Offset Amount=2)
  ]
  rendererClasses [&NiagaraMeshRendererProperties]
  renderers [{rendererClass=&NiagaraMeshRendererProperties, rendererIndex=0}]
 }
 Idle {
  emitterUpdateScript [
   &EmitterState(
    Inactive Response=::NewEnumerator1
    Life Cycle Mode=::NewEnumerator1
    Loop Behavior=::NewEnumerator1
    Loop Delay=0.45
    Loop Duration Mode=::NewEnumerator1
    UseLoopDelay=true
   )
   &SpawnBurst_Instantaneous(Spawn Count=3, Use Spawn Probability=false, UseLoopCountLimit=true)
  ]
  particleSpawnScript [
   &PartitionParticles(Number Of Partitions=3, Particles Per Partition=1, Partition Mode:ENiagara_PartitionMode_v2=::NewEnumerator1)
   &InitializeParticle(Lifetime=0, Ribbon Width=3, Ribbon Width Mode=::NewEnumerator1, UsePositionOffset=false, Write Material Random=true)
   /SetVariables(Particles.RibbonID=/Niagara/DynamicInputs/NiagaraID/MakeNiagaraIDFromInt(
     Index:NiagaraInt32=0
     Acquire Tag:NiagaraInt32=&MakeCustomIntFromBool(
      Bool=&SetBoolByIntComparison(A=$Particles.PartitionParticles.Partition, B=2, Comparison Type=::NewEnumerator6)
      True INT=0
      False INT=1
     )
    )) {isSetParameters=true}
  ]
  particleUpdateScript [
   &ParticleState(Kill Particles When Lifetime Has Elapsed=false)
   &SkeletalMeshLocation(
    Bone / Socket Index=$Particles.PartitionParticles.Partition
    Bone Sampling Mode=::NewEnumerator4
    MaskSpawnGroup=false
    Orientation Sampling=::NewEnumerator1
    Sampled Position Offset=&SelectVectorFromArray(
     Direct Array Index=$Particles.PartitionParticles.Partition
     Vector Selection Array=DataInterface() {propertyValues {FloatData [<3 0 0>, <0 0 0>, <0 0 0>]}}
    )
    Skeletal Mesh=$System.SkeletalMesh
    Use Sampled Mesh Attribute as Ribbon ID=false
   )
   /SetVariables(
    Particles.BladeBlurColor:LinearColor=&MultiplyLinearColor(
     Linear Color=&MakeCustomLinearColorFromBool(
      Bool=$System.IsToon
      True Color=&ColorFromCurve(
       ColorCurve=DataInterface() {propertyValues {
         AlphaCurve {keys [<at 0 1 RCIM_Cubic>,<at 1 1 RCIM_Cubic>]}
         BlueCurve {keys [<at 0 1.5 RCIM_Cubic>,<at 0.1065 0.75>,<at 0.1922 0.9908>,<at 0.2337 2.5152>,<at 0.477 2>,<at 0.5821 2.6976>,<at 0.9167 0.6245>,<at 1 1 RCIM_Cubic>]}
         GreenCurve {keys [<at 0 1.0268 RCIM_Cubic>,<at 0.1065 0.6622>,<at 0.1922 1.0823>,<at 0.2337 2.5152>,<at 0.477 2>,<at 0.5821 2.6954>,<at 0.9167 0.6023>,<at 1 1 RCIM_Cubic>]}
         RedCurve {keys [<at 0 0.842 RCIM_Cubic>,<at 0.1065 0.4937>,<at 0.1922 1.5>,<at 0.2337 2.5152>,<at 0.477 2>,<at 0.5821 2.75>,<at 0.9167 1.1592>,<at 1 1 RCIM_Cubic>]}
        }}
       CurveIndex=Linked() {linkedVariable {name=NPC.FortniteNPC.TimeOfDay_0-1, type=NiagaraFloat}}
      )
      False Color=&ColorFromCurve(
       ColorCurve=DataInterface() {propertyValues {
         AlphaCurve {keys [<at 0 0.9 RCIM_Cubic>]}
         BlueCurve {keys [<at 0 1 RCIM_Cubic>,<at 0.1065 0.75>,<at 0.1922 0.9908>,<at 0.2337 1>,<at 0.477 2>,<at 0.9167 0.6245>,<at 1 1 RCIM_Cubic>]}
         GreenCurve {keys [<at 0 0.6845 RCIM_Cubic>,<at 0.1065 0.6622>,<at 0.1922 1.0823>,<at 0.2337 1>,<at 0.477 2>,<at 0.9167 0.6023>,<at 1 1 RCIM_Cubic>]}
         RedCurve {keys [<at 0 0.5614 RCIM_Cubic>,<at 0.1065 0.4937>,<at 0.1922 1.5>,<at 0.2337 1>,<at 0.477 2>,<at 0.9167 1.1592>,<at 1 0.8388 RCIM_Cubic>]}
        }}
       CurveIndex=Linked() {linkedVariable {name=NPC.FortniteNPC.TimeOfDay_0-1, type=NiagaraFloat}}
      )
     )
     Scale Factor=<0.5 0.5 0.5 1>
    )
    Particles.BladeBlurOrient:Quat4f=$Particles.SkeletalMeshLocation.SampledRotation
    Particles.BladeBlurScale:Vector3f=&VectorFromCurve(
     VectorCurve=DataInterface() {propertyValues {
       XCurve {keys [<at 0 0>,<at 0.25 1.2107>,<at 1 1>]}
       YCurve {keys [<at 0 0>,<at 0.25 1.2107>,<at 1 1>]}
       ZCurve {keys [<at 0 0>,<at 0.25 1.2107>,<at 1 1>]}
      }}
     CurveIndex=$Particles.Age
     Scale Curve=&MergeHelper_VectorSwitch(
      0=<2 2 1>
      1=<1.75 1.75 1>
      2=?
      3=?
      Index=&MakeCustomIntFromBool(Bool=$System.IsToon, True INT=1, False INT=0)
      Num=::NewEnumerator0
     )
    )
    Particles.BladeVlurVis:NiagaraInt32=&MakeCustomIntFromBool(
     Bool=&SetBoolByIntComparison(A=$Particles.PartitionParticles.Partition, B=2, Comparison Type=::NewEnumerator4)
     True INT=2
     False INT=0
    )
   ) {isSetParameters=true}
   /SetVariables(
    Particles.EnergyBallSize:Vector3f=&Multiply_VectorByFloat(
     Vector=<0.1 0.1 0.1>
     Float=&Waveform("Waveform [1]"=::NewEnumerator0, "[1] Frequency"=6, "[1] Amplitude Min/Max"=<1 1.05>)
    )
    Particles.EnergyBallVis:NiagaraInt32=&MakeCustomIntFromBool(
     Bool=&SetBoolByIntComparison(A=$Particles.PartitionParticles.Partition, B=1, Comparison Type=::NewEnumerator6)
     True INT=1
     False INT=0
    )
    Particles.EnergyColor:LinearColor=<rgba 2 2 2 1>
   ) {isSetParameters=true}
   &DynamicMaterialParameters(
    Index 0 Param 1=1
    Index 0 Param 2=0.25
    Index 0 Param 4=1
    Module.Index0Float0Write=true
    Module.Index0Float1Write=true
    Module.Index0Float2Write=true
    Module.Index0Float3Write=true
   )
  ]
  rendererClasses [&NiagaraRibbonRendererProperties, &NiagaraMeshRendererProperties]
  renderers [
   {rendererClass=&NiagaraRibbonRendererProperties}
   {rendererClass=&NiagaraMeshRendererProperties, rendererIndex=1}
   {rendererClass=&NiagaraMeshRendererProperties, rendererIndex=2}
   {rendererClass=&NiagaraMeshRendererProperties, rendererIndex=3}
  ]
 }
}
summary {
 systemName=NS_Pickaxe_DuneBriefShift_Idle_Universal
 emitters [
  {emitterName=Idle, rendererClasses [&NiagaraRibbonRendererProperties, &NiagaraMeshRendererProperties]}
  {emitterName=Equip, rendererClasses [&NiagaraMeshRendererProperties]}
  {emitterName=EquipTracer, rendererClasses [&NiagaraMeshRendererProperties]}
 ]
 userVariables [
  User.bIsFrontEnd {type=NiagaraBool}
  User.bIsFrontEndPreview {type=NiagaraBool}
  User.VariantColor {type=NiagaraFloat, defaultValue=0, description=""}
  User.VariantToon {type=NiagaraFloat, defaultValue=1, description=""}
 ]
}
