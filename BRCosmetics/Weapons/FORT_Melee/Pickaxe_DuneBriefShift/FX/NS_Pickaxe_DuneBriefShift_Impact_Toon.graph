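# Toon-styled impact effect for the DuneBriefShift pickaxe: a short burst of flashes, flares, streaks and sparks, plus a light.
# Flash_Elements and Directional_Elements pack several layers into single emitters, using spawn groups and VisibilityTag to select per-layer values.
# Directional layers, Directional_Flare001 and Toon_Sparks launch along System.DirectionVector, with drag and curl noise on sparks.
# Size and brightness adapt to front-end and time of day; platform checks vary some burst counts.
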NiagaraSystem /BRCosmetics/Weapons/FORT_Melee/Pickaxe_DuneBriefShift/FX/NS_Pickaxe_DuneBriefShift_Impact_Toon
dependencies {
 usedDataInterfaces [
  &NiagaraDataInterfaceCurve
  &NiagaraDataInterfaceColorCurve
  &NiagaraDataInterfacePlatformSet
  &NiagaraDataInterfaceSpriteRendererInfo
  &NiagaraDataInterfaceVector2DCurve
  &NiagaraDataInterfaceArrayFloat3
 ]
 usedDynamicInputs [&Vector4FromCurve, /Game/Effects/Niagara/Modules/Utility/FN_DirectionVectorByPickaxeType]
 usedModules [
  &self:SystemSpawnScript.SystemScriptSource.SystemScriptGraph.NiagaraNodeAssignment_0.SetVariables_#1
  &SystemState
 ]
}
emitters {
 Directional_Elements {
  emitterUpdateScript [
   &EmitterState()
   &SpawnBurst_Instantaneous(Spawn Count=1, Spawn Group=4, Use Spawn Probability=false) {module=SpawnBurst_Instantaneous004}
   &SpawnBurst_Instantaneous(
    Spawn Count=&MakeIntFromFloat(Float=&FN_PickFloatDependOnPlatform(
      TargetPlatform=DataInterface() {propertyValues {Platforms {qualityLevelMask=-29}}}
      TargetPlatformActive=&MakeFloatFromInt(Int=&RandomRangeInt(Minimum=4, Maximum=6))
      TargetPlatformNoActive=&MakeFloatFromInt(Int=&RandomRangeInt(Minimum=12, Maximum=16))
     ))
    Spawn Group=2
    Use Spawn Probability=false
    UseLoopCountLimit=false
   ) {module=SpawnBurst_Instantaneous003}
   &SpawnBurst_Instantaneous(
    Spawn Count=&MakeIntFromFloat(Float=&FN_PickFloatDependOnPlatform(
      TargetPlatform=DataInterface() {propertyValues {Platforms {qualityLevelMask=-29}}}
      TargetPlatformActive=&MakeFloatFromInt(Int=&RandomRangeInt(Minimum=3, Maximum=5))
      TargetPlatformNoActive=&MakeFloatFromInt(Int=&RandomRangeInt(Minimum=8, Maximum=10))
     ))
    Spawn Group=3
    Use Spawn Probability=false
    UseLoopCountLimit=false
   ) {module=SpawnBurst_Instantaneous002}
   &SpawnBurst_Instantaneous(Spawn Count=2, Spawn Group=1, Use Spawn Probability=false, UseLoopCountLimit=false) {module=SpawnBurst_Instantaneous001}
   &SpawnBurst_Instantaneous(Use Spawn Probability=false, UseLoopCountLimit=false)
  ]
  particleSpawnScript [
   &InitializeParticle(
    Color=&MergeHelper_ColorSwitch(
     0=<rgba 2 0.7085 0.1 1>
     1=&RandomRangeLinearColor(Minimum=<rgba 0 0 0 0.7>, Maximum=<rgba 0 0 0 1>)
     2=&ColorFromCurve(
      ColorCurve=DataInterface() {propertyValues {
        AlphaCurve {keys [<at 0 1 RCIM_Cubic>]}
        BlueCurve {keys [
          <at 0.2775 0.132>
          <at 0.3074 2>
          {arriveTangent=-4.3243, interpMode=RCIM_Cubic, leaveTangent=-4.3243, time=0.7173, value=1}
          <at 0.7357 0.1478 RCIM_Cubic>
         ]}
        GreenCurve {keys [
          <at 0.2775 1.088>
          <at 0.3074 2>
          {arriveTangent=-3.2432, interpMode=RCIM_Cubic, leaveTangent=-3.2432, time=0.7173, value=1}
          <at 0.7357 0.6109 RCIM_Cubic>
         ]}
        RedCurve {keys [<at 0.2775 3>,<at 0.3074 2>,<at 0.7173 1 RCIM_Cubic>,<at 0.7357 2 RCIM_Cubic>]}
       }}
      CurveIndex=&RandomRangeFloat(Minimum=0, Maximum=1)
     )
     3=&ColorFromCurve(
      ColorCurve=DataInterface() {propertyValues {
        AlphaCurve {keys [<at 0 1 RCIM_Cubic>]}
        BlueCurve {keys [
          <at 0.2775 0.132>
          <at 0.3074 2>
          {arriveTangent=-4.3243, interpMode=RCIM_Cubic, leaveTangent=-4.3243, time=0.7173, value=1}
          <at 0.7357 0.1478 RCIM_Cubic>
         ]}
        GreenCurve {keys [
          <at 0.2775 1.088>
          <at 0.3074 2>
          {arriveTangent=-3.2432, interpMode=RCIM_Cubic, leaveTangent=-3.2432, time=0.7173, value=1}
          <at 0.7357 0.6109 RCIM_Cubic>
         ]}
        RedCurve {keys [<at 0.2775 3>,<at 0.3074 2>,<at 0.7173 1 RCIM_Cubic>,<at 0.7357 2 RCIM_Cubic>]}
       }}
      CurveIndex=&RandomRangeFloat(Minimum=0, Maximum=1)
     )
     4=<rgba 1 1 1 0.25>
     Index=$Emitter.SpawnGroup
     Num=::NewEnumerator3
    )
    Lifetime=&MergeHelper_FloatSwitch(
     0=0.16
     1=&RandomRangeFloat(Minimum=0.1, Maximum=0.15)
     2=&RandomRangeFloat(Minimum=0.05, Maximum=0.45)
     3=&RandomRangeFloat(Minimum=0.24, Maximum=0.47)
     4=0.32
     Index=$Emitter.SpawnGroup
     Num=::NewEnumerator3
    )
    Sprite Size=&MergeHelper_Vector2Switch(
     0=&FN_FrontEndVector2Modifier(Front End:Vector2f=<60 190>, Gameplay:Vector2f=<80 220>)
     1=&Multiply_Vector2D_ByFloat(Float=$System.FrontEndScale, Vector2D=&UniformRangedVector2D(Minimum=<20 40>, Maximum=<25 70>))
     2=&Multiply_Vector2D_ByFloat(
      Float=&Multiply_Float(
       A=&FN_FrontEndFloatModifier(Gameplay=1, Front End=0.45)
       B=&RandomRangeFloat(Minimum=4.5, Maximum=5)
      )
      Vector2D=<1 1.5>
     )
     3=&Multiply_Vector2D_ByFloat(
      Float=&FN_FrontEndFloatModifier(Gameplay=1, Front End=0.45)
      Vector2D=&RandomRangeVector2D(Minimum=<1.9 3.1487>, Maximum=<3 9.8129>)
     )
     4=&Multiply_Vector2D_ByFloat(Float=$System.FrontEndScale, Vector2D=&RandomRangeVector2D(Minimum=<100 170>, Maximum=<120 220>))
     Index=$Emitter.SpawnGroup
     Num=::NewEnumerator3
    )
    Sprite Size Mode=::NewEnumerator4
    Sprite UV Mode=::NewEnumerator1
    UsePositionOffset=false
    Write Material Random=true
   )
   &ShapeLocation(
    Cylinder Height=&MergeHelper_FloatSwitch(0=0, 1=0, 2=80, 3=?, Index=$Emitter.SpawnGroup, Num=::NewEnumerator1)
    Cylinder Radius=&MergeHelper_FloatSwitch(0=0, 1=0, 2=6, 3=?, Index=$Emitter.SpawnGroup, Num=::NewEnumerator1)
    MaskSpawnGroup=false
    Non Uniform Scale=&VectorFromFloat(Value=$System.FrontEndScale)
    Override Seed=false
    Rotation Axis=$System.DirectionVector
    Rotation Mode=::NewEnumerator1
    Shape Primitive=::NewEnumerator1
    Surface Only=false
    UseOwnerScale=true
   )
   &SubUVAnimation(
    End Frame Range Override=3
    Evaluation Type=::NewEnumerator0
    "Sprite Renderer "=DataInterface() {propertyValues {SpriteRenderer=&self:Directional_Elements_0.NiagaraSpriteRendererProperties_0}}
    SubUV Animation Mode:ENiagara_SubUVLookupModeV2=::NewEnumerator2
    UseEndFrame=true
    UseStartFrame=true
   )
   &AddVelocity(
    Module.Use Falloff Distance=false
    Override Seed=false
    Velocity Mode=::NewEnumerator1
    Velocity Speed=&MergeHelper_FloatSwitch(
     0=0
     1=&Multiply_Float(A=20, B=$System.FrontEndScale)
     2=&Multiply_Float(A=&RandomRangeFloat(Minimum=225, Maximum=300), B=$System.FrontEndScale)
     3=&Multiply_Float(A=&RandomRangeFloat(Minimum=200, Maximum=400), B=$System.FrontEndScale)
     4=0
     Index=$Emitter.SpawnGroup
     Num=::NewEnumerator3
    )
   ) {module=AddVelocity001}
   &AddVelocity(
    Rotation Coordinate Space=::Simulation
    Velocity=$System.DirectionVector
    Velocity Speed Scale=&Multiply_Float(
     A=&MergeHelper_FloatSwitch(
      0=&RandomRangeFloat(Minimum=180, Maximum=220)
      1=100
      2=&RandomRangeFloat(Minimum=300, Maximum=700)
      3=0
      4=150
      Index=$Emitter.SpawnGroup
      Num=::NewEnumerator3
     )
     B=$System.FrontEndScale
    )
   )
   &AddVelocity(Velocity=&MergeHelper_VectorSwitch(
     0=<0 0 0>
     1=<0 0 0>
     2=<0 0 0>
     3=&Multiply_Vector(
      A=&RandomRangeVector(Minimum=<100 0 0>, Maximum=<200 0 0>)
      B=&VectorFromFloat(Value=$System.FrontEndScale)
     )
     4=<0 0 0>
     Index=$Emitter.SpawnGroup
     Num=::NewEnumerator3
    )) {module=AddVelocity002}
   &DynamicMaterialParameters(
    Index 0 Param 1=0.2
    Index 0 Param 2=1
    Index 3 Param 1=1.5
    Index 3 Param 2=1
    Module.Index0Float0Write=true
    Module.Index0Float1Write=true
    Module.Index0Float2Write=true
    Module.Index0Float3Write=true
    Module.Index3Float0Write=true
    Module.Index3Float1Write=true
    Module.Index3Float2Write=true
    Module.Index3Float3Write=true
    Write Parameter Index 3=::NewEnumerator1
   )
   /SetVariables(Particles.DynamicParameterAlt:Vector4f=&MakeVector4(X=1, Y=1, Z=1, W=1)) {isSetParameters=true}
   &CameraOffset(Camera Offset Amount=&MergeHelper_FloatSwitch(0=30, 1=25, 2=20, 3=30, Index=$Emitter.SpawnGroup, Num=::NewEnumerator2))
   /SetVariables() {isSetParameters=true}
   /SetVariables(Particles.VisibilityTag=$Emitter.SpawnGroup) {isSetParameters=true}
  ]
  particleUpdateScript [
   &ParticleState()
   &AccelerationForce(Acceleration=&MergeHelper_VectorSwitch(0=<0 0 0>, 1=<0 0 0>, 2=<0 0 -500>, 3=<0 0 -160>, 4=<0 0 0>, Index=$Particles.VisibilityTag, Num=::NewEnumerator3))
   &Drag(
    Drag=&MergeHelper_FloatSwitch(
     0=0
     1=0
     2=&RandomRangeFloat(Minimum=5, Maximum=8)
     3=&FloatFromCurve(FloatCurve=DataInterface() {propertyValues {Curve {keys [<at 0 1>,<at 1 6>]}}}, Scale Curve=1.2)
     4=3
     Index=$Particles.VisibilityTag
     Num=::NewEnumerator3
    )
    Rotational Drag:NiagaraFloat=&MergeHelper_FloatSwitch(0=0, 1=0, 2=1, 3=1, 4=1, Index=$Particles.VisibilityTag, Num=::NewEnumerator3)
    Use Linear Drag=true
    Use Rotational Drag=true
   )
   &CurlNoiseForce(
    Bias Noise=false
    Noise Strength=&MergeHelper_FloatSwitch(
     0=0
     1=0
     2=0
     3=&RandomRangeFloat(Minimum=200, Maximum=400)
     4=0
     Index=$Particles.VisibilityTag
     Num=::NewEnumerator3
    )
    Pan Noise=false
   )
   &SolveForcesAndVelocity(Clamp Velocity=false, Limit Acceleration=false)
   &ScaleSpriteSize(
    Scale Factor=&MergeHelper_Vector2Switch(
     0=&Vector2DFromCurve(
      Vector2Curve=DataInterface() {propertyValues {
        XCurve {keys [{tangentMode=RCTM_User, time=0, value=0}, <at 0.2 1.075>, {tangentMode=RCTM_User, time=1, value=-0.1}]}
        YCurve {keys [{tangentMode=RCTM_User, time=0, value=0}, <at 0.1 1.025>, {tangentMode=RCTM_User, time=1, value=0.8}]}
       }}
      ScaleCurve=<1 1>
     )
     1=&Vector2DFromCurve(
      Vector2Curve=DataInterface() {propertyValues {
        XCurve {keys [<at 0 1 RCIM_Cubic>,<at 1 0 RCIM_Cubic>]}
        YCurve {keys [{arriveTangent=3.9291, interpMode=RCIM_Cubic, leaveTangent=3.9291, tangentMode=RCTM_User, time=0, value=0}, <at 1 2 RCIM_Cubic>]}
       }}
      ScaleCurve=<1.5 1>
     )
     2=&Vector2DFromFloat(Value=&FloatFromCurve(
       FloatCurve=DataInterface() {propertyValues {Curve {keys [{arriveTangent=-6.725, interpMode=RCIM_Cubic, leaveTangent=-6.725, tangentMode=RCTM_User, time=0.65, value=1}, <at 1 0 RCIM_Cubic>]}}}
       Scale Curve=1.5
      ))
     3=&Vector2DFromFloat(Value=&FloatFromCurve(FloatCurve=DataInterface() {propertyValues {Curve {keys [
           <at 0 0 RCIM_Cubic>
           {interpMode=RCIM_Cubic, tangentMode=RCTM_User, time=0.118, value=1}
           {interpMode=RCIM_Cubic, tangentMode=RCTM_User, time=1, value=0}
          ]}}}))
     4=<1 1>
     Index=$Particles.VisibilityTag
     Num=::NewEnumerator3
    )
    Scale Sprite Size Mode=::NewEnumerator2
   )
   &ScaleSpriteSizeBySpeed(
    Max Scale Factor=&MergeHelper_Vector2Switch(0=<1 1>, 1=<1 1>, 2=<0.2 5>, 3=<1 6>, 4=<1 1>, Index=$Particles.VisibilityTag, Num=::NewEnumerator3)
    Min Scale Factor=&MergeHelper_Vector2Switch(0=<1 1>, 1=<1 1>, 2=<0.2 0.2>, 3=<1 3>, 4=<1 1>, Index=$Particles.VisibilityTag, Num=::NewEnumerator3)
   )
   /SetVariables(Particles.DynamicParameterAlt1:Vector4f=&MakeVector4(
     X=&FloatFromCurve(FloatCurve=DataInterface() {propertyValues {Curve {keys [<at 0 0>,<at 1 1>]}}})
     Y=&UniformAOrBFloat(A=0, Evaluation Type=::NewEnumerator0, Recalculate Random Each Loop=true, Override Seed=false)
     Z=0
     W=0
    )) {isSetParameters=true}
   &ScaleColor(
    Scale Alpha=&MergeHelper_FloatSwitch(
     0=$System.ToDScalar
     1=&FloatFromCurve(FloatCurve=DataInterface() {propertyValues {Curve {keys [<at 0 1 RCIM_Cubic>,<at 1 0>]}}}, Scale Curve=$System.ToDScalar)
     2=$System.ToDScalar
     3=&FloatFromCurve(
      FloatCurve=DataInterface() {propertyValues {Curve {keys [
          {arriveTangent=-0.0852, interpMode=RCIM_Cubic, leaveTangent=-0.0852, tangentMode=RCTM_User, time=0, value=1}
          {arriveTangent=-2.6621, interpMode=RCIM_Cubic, leaveTangent=-2.6621, tangentMode=RCTM_User, time=1, value=0}
         ]}}}
      Scale Curve=$System.ToDScalar
     )
     4=&FloatFromCurve(
      FloatCurve=DataInterface() {propertyValues {Curve {keys [<at 0 1>,<at 0.8091 0.9932>,<at 1 0>]}}}
      Scale Curve=&FNTimeOfDayScalar(Night Scale=5, Morning Scale=4, Day Scale=2, Evening Scale=4)
     )
     Index=$Particles.VisibilityTag
     Num=::NewEnumerator3
    )
    Scale RGB=&MergeHelper_VectorSwitch(
     0=<1 1 1>
     1=<1 1 1>
     2=&VectorFromFloat(Value=&FloatFromCurve(
       FloatCurve=DataInterface() {propertyValues {Curve {keys [
           {interpMode=RCIM_Cubic, tangentMode=RCTM_User, time=0, value=1}
           <at 0.3 0.2>
           {interpMode=RCIM_Cubic, tangentMode=RCTM_User, time=1, value=0}
          ]}}}
       Scale Curve=10
      ))
     3=&VectorFromFloat(Value=&FloatFromCurve(FloatCurve=DataInterface() {propertyValues {Curve {keys [
           {interpMode=RCIM_Cubic, tangentMode=RCTM_User, time=0, value=1}
           {arriveTangent=-1.4115, interpMode=RCIM_Cubic, leaveTangent=-1.4115, tangentMode=RCTM_User, time=1, value=0}
          ]}}}))
     4=&MakeVectorFromLinearColorRGB(Color=&FNTimeOfDayColor(
       Night Color=<rgba 0.34 0.51 0.85 1>
       Morning Color=<rgba 0.902 0.5882 0.5412 1>
       Day Color=<rgba 1 1 1 1>
       Evening Color=<rgba 0.9 0.54 0.6 1>
      ))
     Index=$Particles.VisibilityTag
     Num=::NewEnumerator3
    )
    ScaleA=true
    ScaleRGB=true
   )
  ]
  rendererClasses [&NiagaraSpriteRendererProperties]
  renderers [
   {rendererClass=&NiagaraSpriteRendererProperties}
   {rendererClass=&NiagaraSpriteRendererProperties, rendererIndex=1}
   {rendererClass=&NiagaraSpriteRendererProperties, rendererIndex=2}
   {rendererClass=&NiagaraSpriteRendererProperties, rendererIndex=3}
   {rendererClass=&NiagaraSpriteRendererProperties, rendererIndex=4}
  ]
 }
 Directional_Flare001 {
  emitterUpdateScript [
   &EmitterState(
    Enable Distance Culling=true
    Life Cycle Mode=::NewEnumerator1
    Loop Behavior=::NewEnumerator1
    Loop Delay=0.01
    Loop Duration=1
    Max Distance Response=::KillImmediately
    MaxDistance=3500
    Module.bUseMaxDistance=true
    Module.bUseMinDistance=false
    Scalability Mode=::NewEnumerator1
    UseLoopDelay=false
   )
   &SpawnBurst_Instantaneous(Spawn Count=1, Use Spawn Probability=true)
  ]
  particleSpawnScript [
   &InitializeParticle(
    Color=<rgba 5 4.2819 1.3169 7>
    Lifetime=0.2
    Sprite Size=<50 175>
    Sprite Size Mode=::NewEnumerator4
    UsePositionOffset=true
    Write Material Random=true
   ) {module=InitializeParticle001}
   &CameraOffset(Camera Offset Amount=&FN_FrontEndFloatModifier(Gameplay=40, Front End=10))
   &DynamicMaterialParameters(
    Index 0 Param 1=0.2
    Index 0 Param 2=1
    Module.Param0WriteEnabled=true
    Module.Param1WriteEnabled=true
    Module.Param2WriteEnabled=true
    Module.Param3WriteEnabled=true
   )
   &AddVelocity(
    Velocity=$System.DirectionVector
    Velocity Speed Scale=&FN_FrontEndFloatModifier(Gameplay=200, Front End=50)
   )
  ]
  particleUpdateScript [
   &ParticleState()
   &SolveForcesAndVelocity(Clamp Velocity=false, Limit Acceleration=false)
   &ScaleColor(
    Color Value To Scale=$Particles.Color
    Scale Alpha=&FloatFromCurve(
     FloatCurve=DataInterface() {propertyValues {Curve {keys [
         <at 0 0 RCIM_Cubic>
         {arriveTangent=-2.137, interpMode=RCIM_Cubic, leaveTangent=-2.137, tangentMode=RCTM_User, time=0.06, value=1}
         {arriveTangent=-0.4517, interpMode=RCIM_Cubic, leaveTangent=-0.4517, tangentMode=RCTM_User, time=1, value=0}
        ]}}}
     Scale Curve=&FN_FrontEndFloatModifier(Gameplay=1, Front End=0.6)
    )
    Scale RGB=&VectorFromFloat(Value=&Lerp_Float(
      A=1.5
      B=0.6
      Alpha=&FloatFromCurve(FloatCurve=DataInterface() {propertyValues {Curve {keys [
           {arriveTangent=13.3064, interpMode=RCIM_Cubic, leaveTangent=13.3064, tangentMode=RCTM_User, time=0.02, value=0}
           {arriveTangent=1.5312, interpMode=RCIM_Cubic, leaveTangent=1.5312, tangentMode=RCTM_User, time=0.1, value=0.9}
           {arriveTangent=0.0212, interpMode=RCIM_Cubic, leaveTangent=0.0212, tangentMode=RCTM_User, time=0.4, value=1}
          ]}}})
     ))
    ScaleA=true
    ScaleRGB=true
   )
   &ScaleSpriteSize(
    Non-Uniform Curve Sprite Scale=DataInterface() {propertyValues {
      XCurve {keys [
        {interpMode=RCIM_Cubic, tangentMode=RCTM_User, time=0, value=0.2}
        {arriveTangent=3.4491, interpMode=RCIM_Cubic, leaveTangent=3.4491, tangentMode=RCTM_User, time=0.08, value=0.3}
        {arriveTangent=-2.6795, interpMode=RCIM_Cubic, leaveTangent=-2.6795, tangentMode=RCTM_User, time=0.11, value=1.2}
        {arriveTangent=-1.5783, interpMode=RCIM_Cubic, leaveTangent=-1.5783, tangentMode=RCTM_User, time=0.2, value=0.6}
        {arriveTangent=-0.4384, interpMode=RCIM_Cubic, leaveTangent=-0.4384, tangentMode=RCTM_User, time=1, value=0}
       ]}
      YCurve {keys [
        {interpMode=RCIM_Cubic, tangentMode=RCTM_User, time=0, value=0.2}
        {interpMode=RCIM_Cubic, tangentMode=RCTM_User, time=0.1, value=1.025}
        {tangentMode=RCTM_User, time=1, value=0.8}
       ]}
     }}
    Scale Sprite Size Mode=::NewEnumerator3
   ) {module=ScaleSpriteSize001}
   /SetVariables(
    Particles.ColorDark:LinearColor=&MultiplyLinearColor(
     Linear Color=&ColorFromCurve(ColorCurve=DataInterface() {propertyValues {
        AlphaCurve {keys [<at 0 3 RCIM_Cubic>,<at 0.1566 1>,<at 0.519 0 RCIM_Cubic>]}
        BlueCurve {keys [<at 0.0459 0.6224 RCIM_Cubic>,<at 0.5447 0.6701>]}
        GreenCurve {keys [<at 0.0459 0.5746 RCIM_Cubic>,<at 0.5447 0.5548>]}
        RedCurve {keys [<at 0.0459 0.516 RCIM_Cubic>,<at 0.5447 0.4945>]}
       }})
     Scale Factor=&MakeFloat4FromVectorAndFloat("Vector (RGB)"=<1 1 1>, "Float (Alpha)"=&FN_FrontEndFloatModifier(Gameplay=1, Front End=0.8))
    )
    Particles.SpriteSizeDark:Vector2f=&Multiply_Vector2D(
     A=$Particles.SpriteSize
     B=&FN_FrontEndVector2Modifier(Front End:Vector2f=<0.65 2>, Gameplay:Vector2f=<0.65 2>)
    )
   ) {isSetParameters=true}
   /SetVariables(Particles.SpriteSizeGlow:Vector2f=&Multiply_Vector2D(
     A=$Particles.SpriteSize
     B=&MakeVector2D(
      X=0.7
      Y=&FloatFromCurve(FloatCurve=DataInterface() {propertyValues {Curve {keys [
           {arriveTangent=5.9595, interpMode=RCIM_Cubic, leaveTangent=5.9595, tangentMode=RCTM_User, time=0, value=0.55}
           {interpMode=RCIM_Cubic, tangentMode=RCTM_User, time=0.6, value=2}
          ]}}})
     )
    )) {isSetParameters=true}
  ]
  rendererClasses [&NiagaraSpriteRendererProperties]
  renderers [
   {rendererClass=&NiagaraSpriteRendererProperties}
   {rendererClass=&NiagaraSpriteRendererProperties, rendererIndex=1}
  ]
 }
 Flash_Elements {
  emitterUpdateScript [
   &EmitterState(
    Enable Distance Culling=true
    Max Distance Response=::KillImmediately
    Module.bUseMaxDistance=true
    Module.bUseMinDistance=false
    Scalability Mode=::NewEnumerator1
   )
   &SpawnBurst_Instantaneous(Spawn Count=1, Spawn Group=2, Use Spawn Probability=false) {module=SpawnBurst_Instantaneous003}
   &SpawnBurst_Instantaneous(Spawn Count=1, Spawn Group=1, Use Spawn Probability=false) {module=SpawnBurst_Instantaneous002}
   &SpawnBurst_Instantaneous(Use Spawn Probability=false, UseLoopCountLimit=false)
  ]
  particleSpawnScript [
   &InitializeParticle(
    Color=&MergeHelper_ColorSwitch(0=<rgba 1 1 1 1>, 1=<rgba 1 1 1 1>, 2=<rgba 1 1 1 1>, 3=?, Index=$Emitter.SpawnGroup, Num=::NewEnumerator1)
    Lifetime=&MergeHelper_FloatSwitch(0=0.133, 1=0.2, 2=0.3, 3=?, Index=$Emitter.SpawnGroup, Num=::NewEnumerator1)
    Sprite Rotation Mode=::NewEnumerator2
    Sprite Size Mode=::NewEnumerator3
    Uniform Sprite Size=&Multiply_Float(
     A=&MergeHelper_FloatSwitch(
      0=&UniformRangedFloat(Minimum=100, Maximum=125)
      1=200
      2=&Multiply_Float(A=140, B=&RandomRangeFloat(Minimum=0.7, Maximum=1))
      3=?
      Index=$Emitter.SpawnGroup
      Num=::NewEnumerator1
     )
     B=$System.FrontEndScale
    )
    UsePositionOffset=false
    Write Material Random=true
   )
   &CameraOffset(Camera Offset Amount=&MergeHelper_FloatSwitch(0=10, 1=25, 2=10, 3=?, Index=$Emitter.SpawnGroup, Num=::NewEnumerator1))
   /SetVariables(Particles.DynamicParameterAlt2:Vector4f=&MakeVector4(X=1, Y=1, Z=0, W=0)) {isSetParameters=true}
   /SetVariables() {isSetParameters=true}
   /SetVariables(Particles.VisibilityTag=$Emitter.SpawnGroup) {isSetParameters=true}
  ]
  particleUpdateScript [
   &ParticleState()
   &ScaleSpriteSize(
    Scale Sprite Size Mode=::NewEnumerator0
    Uniform Scale Factor=&MergeHelper_FloatSwitch(
     0=&FloatFromCurve(FloatCurve=DataInterface() {propertyValues {Curve {keys [
          {arriveTangent=2.105, interpMode=RCIM_Cubic, leaveTangent=2.105, tangentMode=RCTM_User, time=0, value=0}
          {arriveTangent=0.0224, interpMode=RCIM_Cubic, leaveTangent=0.0224, tangentMode=RCTM_User, time=1, value=1}
         ]}}})
     1=&FloatFromCurve(FloatCurve=DataInterface() {propertyValues {Curve {keys [
          {arriveTangent=1.6131, interpMode=RCIM_Cubic, leaveTangent=1.6131, tangentMode=RCTM_User, time=0, value=0.7}
          {arriveTangent=0.0224, interpMode=RCIM_Cubic, leaveTangent=0.0224, tangentMode=RCTM_User, time=1, value=0}
         ]}}})
     2=&FloatFromCurve(FloatCurve=DataInterface() {propertyValues {Curve {keys [{arriveTangent=3.9086, interpMode=RCIM_Cubic, leaveTangent=3.9086, tangentMode=RCTM_User, time=0, value=0}, <at 1 1.4 RCIM_Cubic>]}}})
     3=?
     Index=$Particles.VisibilityTag
     Num=::NewEnumerator1
    )
   ) {module=ScaleSpriteSize001}
   &ScaleColor(
    Scale Alpha=&MergeHelper_FloatSwitch(
     0=&FloatFromCurve(
      FloatCurve=DataInterface() {propertyValues {Curve {keys [
          {arriveTangent=9.0725, interpMode=RCIM_Cubic, leaveTangent=9.0725, tangentMode=RCTM_User, time=0, value=0}
          {arriveTangent=0.3975, interpMode=RCIM_Cubic, leaveTangent=0.3975, tangentMode=RCTM_User, time=0.0555, value=1}
          {arriveTangent=-0.0405, interpMode=RCIM_Cubic, leaveTangent=-0.0405, tangentMode=RCTM_User, time=1, value=0}
         ]}}}
      Scale Curve=$System.ToDScalar
     )
     1=$System.ToDScalar
     2=&FloatFromCurve(
      FloatCurve=DataInterface() {propertyValues {Curve {keys [
          {arriveTangent=-0.2921, interpMode=RCIM_Cubic, leaveTangent=-0.2921, tangentMode=RCTM_User, time=0, value=7}
          {arriveTangent=-7, interpMode=RCIM_Cubic, leaveTangent=-7, time=0.4, value=4}
          <at 1 0 RCIM_Cubic>
         ]}}}
      Scale Curve=2
     )
     3=?
     Index=$Particles.VisibilityTag
     Num=::NewEnumerator1
    )
    Scale RGB=&MergeHelper_VectorSwitch(
     0=&VectorFromFloat(Value=&FloatFromCurve(
       FloatCurve=DataInterface() {propertyValues {Curve {keys [
           <at 0 10 RCIM_Cubic>
           {arriveTangent=-9, interpMode=RCIM_Cubic, leaveTangent=-9, time=0.5, value=3}
           {tangentMode=RCTM_User, time=1, value=1}
          ]}}}
       Scale Curve=$System.ToDScalar
      ))
     1=&MakeVectorFromLinearColorRGB(Color=&ColorFromCurve(ColorCurve=DataInterface() {propertyValues {
         AlphaCurve {keys [<at 0 1 RCIM_Cubic>]}
         BlueCurve {keys [<at 0 0.1678>,<at 0.7773 0.0816>]}
         GreenCurve {keys [<at 0 0.5146>,<at 0.7773 0.2211>]}
         RedCurve {keys [<at 0 1>,<at 0.7773 0.5>]}
        }}))
     2=<1 1 1>
     3=?
     Index=$Particles.VisibilityTag
     Num=::NewEnumerator1
    )
    ScaleA=true
    ScaleRGB=true
   )
   /SetVariables(
    Particles.InnerColor:LinearColor=&MultiplyLinearColor(
     Linear Color=<rgba 10 5.9116 0.7305 1>
     Scale Factor=&MakeVector4_FromLinearColor(LinearColor=$Particles.Color)
    )
    Particles.InnerDynamicParameter:Vector4f=&MakeVector4_FromFloat(Float=&Lerp_Float(A=0, B=0.5, Alpha=$Particles.NormalizedAge))
    Particles.InnerSize:Vector2f=&Multiply_Vector2D_ByFloat(Float=0.5, Vector2D=$Particles.SpriteSize)
    Particles.MiddleColor:LinearColor=&MultiplyLinearColor(
     Linear Color=<rgba 5 1.0417 0.25 1>
     Scale Factor=&MakeVector4_FromLinearColor(LinearColor=$Particles.Color)
    )
    Particles.MiddleDynamicParameter:Vector4f=&MakeVector4_FromFloat(Float=&Lerp_Float(A=0.25, B=0.75, Alpha=$Particles.NormalizedAge))
    Particles.MiddleSize:Vector2f=&Multiply_Vector2D_ByFloat(Float=0.75, Vector2D=$Particles.SpriteSize)
    Particles.OuterColor:LinearColor=&MultiplyLinearColor(Linear Color=<rgba 0 0 0 1>, Scale Factor=&MakeVector4_FromLinearColor(LinearColor=$Particles.Color))
    Particles.OuterDynamicParameter:Vector4f=&MakeVector4_FromFloat(Float=&Lerp_Float(A=0, B=1, Alpha=$Particles.NormalizedAge))
    Particles.OuterSize:Vector2f=&Multiply_Vector2D_ByFloat(Float=1.3, Vector2D=$Particles.SpriteSize)
    Particles.SubImageIndex=&RandomRangeFloat(Minimum=0, Maximum=4)
    Particles.SubImageIndexAlt:NiagaraFloat=&RandomRangeFloat(Minimum=0, Maximum=3)
   ) {isSetParameters=true}
   /SetVariables(Particles.DynamicParameterAlt3:Vector4f=&MakeVector4(
     X=&FloatFromCurve(FloatCurve=DataInterface() {propertyValues {Curve {keys [<at 0 1 RCIM_Cubic>, {arriveTangent=-2.742, interpMode=RCIM_Cubic, leaveTangent=-2.742, tangentMode=RCTM_User, time=1, value=0}]}}})
     Y=&FloatFromCurve(FloatCurve=DataInterface() {propertyValues {Curve {keys [<at 0 0 RCIM_Cubic>, {arriveTangent=2.4833, interpMode=RCIM_Cubic, leaveTangent=2.4833, tangentMode=RCTM_User, time=1, value=1}]}}})
     Z=0
     W=0
    )) {isSetParameters=true}
   &FN_Light_Attributes(
    Color=<rgba 1 0.4667 0 1>
    Create Color Attribute=true
    Create Exponent Attribute=true
    Create Position Attribute=true
    Create Radius Attribute=true
    Exponent=5
    Position=<15 0 0>
    Radius=90
    Scale Color=&FloatFromCurve(
     FloatCurve=DataInterface() {propertyValues {Curve {keys [
         {interpMode=RCIM_Cubic, tangentMode=RCTM_User, time=0, value=1}
         {arriveTangent=-0.5426, interpMode=RCIM_Cubic, leaveTangent=-0.5426, tangentMode=RCTM_User, time=0.8333, value=0}
        ]}}}
     Scale Curve=10
    )
   )
  ]
  rendererClasses [&NiagaraSpriteRendererProperties, &NiagaraLightRendererProperties]
  renderers [
   {rendererClass=&NiagaraSpriteRendererProperties}
   {rendererClass=&NiagaraSpriteRendererProperties, rendererIndex=1}
   {rendererClass=&NiagaraSpriteRendererProperties, rendererIndex=2}
   {rendererClass=&NiagaraLightRendererProperties, rendererIndex=3}
   {rendererClass=&NiagaraSpriteRendererProperties, rendererIndex=4}
   {rendererClass=&NiagaraSpriteRendererProperties, rendererIndex=5}
  ]
 }
 Toon_Sparks {
  emitterUpdateScript [
   &EmitterState()
   &SpawnBurst_Instantaneous(
    Spawn Count=&RandomRangeInt(Minimum=8, Maximum=12)
    Use Spawn Probability=false
    UseLoopCountLimit=false
   )
  ]
  particleSpawnScript [
   &InitializeParticle(
    Color=&ColorFromCurve(
     ColorCurve=DataInterface() {propertyValues {
       AlphaCurve {keys [<at 0 1 RCIM_Cubic>]}
       BlueCurve {keys [
         <at 0.2775 0.132>
         <at 0.3074 2>
         {arriveTangent=-4.3243, interpMode=RCIM_Cubic, leaveTangent=-4.3243, time=0.7173, value=1}
         <at 0.7357 0.1478 RCIM_Cubic>
        ]}
       GreenCurve {keys [
         <at 0.2775 1.088>
         <at 0.3074 2>
         {arriveTangent=-3.2432, interpMode=RCIM_Cubic, leaveTangent=-3.2432, time=0.7173, value=1}
         <at 0.7357 0.6109 RCIM_Cubic>
        ]}
       RedCurve {keys [<at 0.2775 3>,<at 0.3074 2>,<at 0.7173 1 RCIM_Cubic>,<at 0.7357 2 RCIM_Cubic>]}
      }}
     CurveIndex=&RandomRangeFloat(Minimum=0, Maximum=1)
    )
    Lifetime=&RandomRangeFloat(Minimum=0.24, Maximum=0.47)
    Sprite Size Max=<3 12>
    Sprite Size Min=<1.9 5>
    Sprite Size Mode=::NewEnumerator2
    UsePositionOffset=false
    Write Material Random=true
   )
   &AddVelocityFromPoint(
    Module.Use Falloff Distance=false
    Velocity Strength=&Multiply_Float(
     A=&RandomRangeFloat(Minimum=50, Maximum=75)
     B=&FN_FrontEndFloatModifier(Gameplay=1, Front End=$System.FrontEndScale)
    )
   )
   &AddVelocity(Velocity=$System.DirectionVector, Velocity Speed Scale=&RandomRangeFloat(Minimum=200, Maximum=400))
  ]
  particleUpdateScript [
   &ParticleState()
   &ScaleSpriteSize(
    Uniform Curve Scale=&FN_FrontEndFloatModifier(Gameplay=1, Front End=$System.FrontEndScale)
    Uniform Curve Sprite Scale=DataInterface() {propertyValues {Curve {keys [
        <at 0 0 RCIM_Cubic>
        {interpMode=RCIM_Cubic, tangentMode=RCTM_User, time=0.118, value=1}
        {interpMode=RCIM_Cubic, tangentMode=RCTM_User, time=1, value=0}
       ]}}}
   )
   &ScaleColor(
    Scale Alpha=&FloatFromCurve(
     FloatCurve=DataInterface() {propertyValues {Curve {keys [
         {arriveTangent=-0.0852, interpMode=RCIM_Cubic, leaveTangent=-0.0852, tangentMode=RCTM_User, time=0, value=1}
         {arriveTangent=-2.6621, interpMode=RCIM_Cubic, leaveTangent=-2.6621, tangentMode=RCTM_User, time=1, value=0}
        ]}}}
     Scale Curve=$System.ToDScalar
    )
    Scale RGB=&VectorFromFloat(Value=&FloatFromCurve(FloatCurve=DataInterface() {propertyValues {Curve {keys [
          {interpMode=RCIM_Cubic, tangentMode=RCTM_User, time=0, value=1}
          {arriveTangent=-1.4115, interpMode=RCIM_Cubic, leaveTangent=-1.4115, tangentMode=RCTM_User, time=1, value=0}
         ]}}}))
    ScaleA=true
    ScaleRGB=true
   )
   &Drag(
    Drag=&FloatFromCurve(FloatCurve=DataInterface() {propertyValues {Curve {keys [<at 0 1>,<at 1 6>]}}}, Scale Curve=1.2)
    Use Linear Drag=true
    Use Rotational Drag=true
   )
   &CurlNoiseForce(Bias Noise=false, Noise Strength=&RandomRangeFloat(Minimum=200, Maximum=400), Pan Noise=true)
   &AccelerationForce(Acceleration=<0 0 -160>)
   &SolveForcesAndVelocity(Clamp Velocity=false, Limit Acceleration=false)
   &ScaleSpriteSizeBySpeed(Max Scale Factor=<1 6>, Min Scale Factor=<1 3>)
  ]
  rendererClasses [&NiagaraSpriteRendererProperties]
  renderers [{rendererClass=&NiagaraSpriteRendererProperties}]
 }
}
summary {
 systemName=NS_Pickaxe_DuneBriefShift_Impact_Toon
 emitters [
  {
   emitterName=Flash_Elements
   rendererClasses [&NiagaraSpriteRendererProperties, &NiagaraLightRendererProperties]
  }
  {emitterName=Directional_Elements, rendererClasses [&NiagaraSpriteRendererProperties]}
  {emitterName=Toon_Sparks, rendererClasses [&NiagaraSpriteRendererProperties]}
  {emitterName=Directional_Flare001, rendererClasses [&NiagaraSpriteRendererProperties]}
 ]
 userVariables [
  User.bIsFrontEnd {type=NiagaraBool}
  User.bIsFrontEndPreview {type=NiagaraBool}
  User.PhysicalSurface {type=NiagaraInt32}
  User.SwingSectionIndex {type=NiagaraInt32}
  User.VariantColor {type=NiagaraFloat, defaultValue=0, description=""}
  User.VariantToon {type=NiagaraFloat, defaultValue=0, description=""}
 ]
}
