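# Niagara swing trail for the DuneBriefShift pickaxe cosmetic, drawn as a short-lived ribbon by a single Trail emitter.
# FN_SpawnAnimTrail spawns particles, FN_InitializeAnimTrail and its bone-index module set them up, and FN_UpdateAnimTrail updates them.
# Trail color is sampled from RGB curves indexed by Fortnite time of day, with System.IsToon choosing between toon and standard curve sets.
# DynamicMaterialParameters feeds values to the ribbon material.
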NiagaraSystem /BRCosmetics/Weapons/FORT_Melee/Pickaxe_DuneBriefShift/FX/NS_Pickaxe_DuneBriefShift_Trail
dependencies {
 usedDataInterfaces [&NiagaraDataInterfaceColorCurve, &NiagaraDataInterfaceSkeletalMesh]
 usedDynamicInputs [&SetBoolByFloatComparison]
 usedModules [
  &self:SystemSpawnScript.SystemScriptSource.SystemScriptGraph.NiagaraNodeAssignment_0.SetVariables_#1
  &SystemState
  &self:SystemSpawnScript.SystemScriptSource.SystemScriptGraph.NiagaraNodeAssignment_1.SetVariables_#2
  &self:SystemSpawnScript.SystemScriptSource.SystemScriptGraph.NiagaraNodeAssignment_2.SetVariables_#3
 ]
}
emitters {Trail {
  emitterUpdateScript [
   &EmitterState()
   &FN_SpawnAnimTrail(Switch on ENiagara_AnimTrail_Source:ENiagara_AnimTrail_Source=::NewEnumerator1)
  ]
  particleSpawnScript [
   &InitializeParticle(Color=<rgba 0.0885 0.0885 0.0885 0.5>, Lifetime=5, Ribbon Width Mode=::NewEnumerator1, UsePositionOffset=false, Write Material Random=true)
   &FN_InitializeAnimTrailBoneIndecies()
   &FN_InitializeAnimTrail(
    Color=&MakeCustomLinearColorFromBool(
     Bool=$System.IsToon
     True Color=&ColorFromCurve(
      ColorCurve=DataInterface() {propertyValues {
        AlphaCurve {keys [<at 0 0.25 RCIM_Cubic>]}
        BlueCurve {keys [<at 0 1.5 RCIM_Cubic>,<at 0.1065 0.75>,<at 0.1922 0.9908>,<at 0.2337 2.5152>,<at 0.477 2>,<at 0.9167 0.6245>,<at 1 1 RCIM_Cubic>]}
        GreenCurve {keys [<at 0 1.0268 RCIM_Cubic>,<at 0.1065 0.6622>,<at 0.1922 1.0823>,<at 0.2337 2.5152>,<at 0.477 2>,<at 0.9167 0.6023>,<at 1 1 RCIM_Cubic>]}
        RedCurve {keys [<at 0 0.842 RCIM_Cubic>,<at 0.1065 0.4937>,<at 0.1922 1.5>,<at 0.2337 2.5152>,<at 0.477 2>,<at 0.9167 1.1592>,<at 1 1 RCIM_Cubic>]}
       }}
      CurveIndex=Linked() {linkedVariable {name=NPC.FortniteNPC.TimeOfDay_0-1, type=NiagaraFloat}}
     )
     False Color=&ColorFromCurve(
      ColorCurve=DataInterface() {propertyValues {
        AlphaCurve {keys [<at 0 0.25 RCIM_Cubic>]}
        BlueCurve {keys [<at 0 1 RCIM_Cubic>,<at 0.1065 0.75>,<at 0.1922 0.9908>,<at 0.2337 1>,<at 0.477 2>,<at 0.9167 0.6245>,<at 1 1 RCIM_Cubic>]}
        GreenCurve {keys [<at 0 0.6845 RCIM_Cubic>,<at 0.1065 0.6622>,<at 0.1922 1.0823>,<at 0.2337 1>,<at 0.477 2>,<at 0.9167 0.6023>,<at 1 1 RCIM_Cubic>]}
        RedCurve {keys [<at 0 0.5614 RCIM_Cubic>,<at 0.1065 0.4937>,<at 0.1922 1.5>,<at 0.2337 1>,<at 0.477 2>,<at 0.9167 1.1592>,<at 1 0.8388 RCIM_Cubic>]}
       }}
      CurveIndex=Linked() {linkedVariable {name=NPC.FortniteNPC.TimeOfDay_0-1, type=NiagaraFloat}}
     )
    )
    Lifetime=0.3
   )
  ]
  particleUpdateScript [
   &ParticleState()
   &FN_UpdateAnimTrail()
   &DynamicMaterialParameters(
    Index 2 Param 1=4
    Index 2 Param 2=2
    Module.Index0Float0Write=true
    Module.Index0Float1Write=true
    Module.Index0Float2Write=true
    Module.Index0Float3Write=true
    Module.Index1Float0Write=true
    Module.Index1Float1Write=true
    Module.Index1Float2Write=true
    Module.Index1Float3Write=true
    Module.Index2Float0Write=true
    Module.Index2Float1Write=true
    Module.Index2Float2Write=true
    Module.Index2Float3Write=true
    Write Parameter Index 1=::NewEnumerator1
    Write Parameter Index 2=::NewEnumerator1
   )
  ]
  rendererClasses [&NiagaraRibbonRendererProperties]
  renderers [{rendererClass=&NiagaraRibbonRendererProperties}]
 }}
summary {
 systemName=NS_Pickaxe_DuneBriefShift_Trail
 emitters [{emitterName=Trail, rendererClasses [&NiagaraRibbonRendererProperties]}]
 userVariables [
  User.bIsFrontEnd {type=NiagaraBool}
  User.bIsFrontEndPreview {type=NiagaraBool}
  User.RibbonIDOffset {type=NiagaraInt32}
  User.SwingSectionIndex {type=NiagaraInt32}
  User.VariantColor {type=NiagaraFloat, defaultValue=0, description=""}
  User.VariantToon {type=NiagaraFloat, defaultValue=0, description=""}
 ]
}
