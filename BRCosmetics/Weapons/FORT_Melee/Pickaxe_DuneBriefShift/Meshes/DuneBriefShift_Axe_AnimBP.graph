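# Anim Blueprint for the DuneBriefShift pickaxe that deploys and spins its blades after a weapon change.
# OnWeaponChanged starts a timer, and two float curves sampled each update drive the bone-modifier alphas until the second reaches 1.
# The first fades out offset and scale on MSTR_machineblades and FX_end over half a second.
# The second blends in continuous real-time pitch rotation on the blades bone over 1.5 seconds.
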AnimBlueprint /BRCosmetics/Weapons/FORT_Melee/Pickaxe_DuneBriefShift/Meshes/DuneBriefShift_Axe_AnimBP
target_skeleton=&/DuneBriefShift_Axe_Skeleton
graphs [
 AnimGraph {nodes [
   AnimGraphNode_SubInput_0 = Input Pose()
   n1 = "Transform (Modify) Bone - Bone: MSTR_machineblades"(
    ComponentPose=Local To Component(LocalPose=@AnimGraphNode_SubInput_0)
    Alpha=Get Alpha() -> (Alpha:float)
    Translation="-44,0,0"
    Rotation="0,0,0"
    Scale="0.1,0.1,0.1"
   ) {
    scaleMode=BMM_Additive
    scaleSpace=BCS_BoneSpace
    translationMode=BMM_Additive
    translationSpace=BCS_BoneSpace
    boneToModify {boneName=MSTR_machineblades}
    scale=<0.1 0.1 0.1>
    translation {x=-44}
   }
   n2 = "Transform (Modify) Bone - Bone: FX_end"(ComponentPose=@n1, Alpha=Get Alpha(), Translation="-26,0,0", Rotation="0,0,0", Scale="0.1,0.1,0.1") {
    scaleMode=BMM_Additive
    scaleSpace=BCS_BoneSpace
    translationMode=BMM_Additive
    translationSpace=BCS_BoneSpace
    boneToModify {boneName=FX_end}
    scale=<0.1 0.1 0.1>
    translation {x=-26}
   }
   n3 = Get move baldes() -> (move baldes:Rotator)
   n4 = Get Alpha_0() -> (Alpha_0:float)
   Output Pose(Result=Component To Local(ComponentPose="Transform (Modify) Bone - Bone: blades"(ComponentPose=@n2, Alpha=@n4, Translation="0,0,0", Rotation=@n3, Scale="1,1,1") {rotationMode=BMM_Additive, rotationSpace=BCS_BoneSpace, boneToModify {boneName=blades}})) {layerGroup=None}
  ]}
 EventGraph {nodes [
   n1 = Event BlueprintUpdateAnimation()
   > Set Delta Time X(Delta Time X=@n1.DeltaTimeX)
   > n2 = Sequence!(execute) -> (then_0, then_1, then_2:Exec)
   > n3 = Cast To FortWeaponPickaxeAthena!(execute, Object=GetOwningActor()) -> (then, CastFailed, AsFort Weapon Pickaxe Athena:Fort Weapon Pickaxe Athena Object Reference) {is=title K2Node_DynamicCast}
   n4 = GetOwner(self:Actor Object Reference=@n3)
   > Is Valid(InputObject=@n4)
   n5 = GetComponentByClass!(self:Actor Object Reference=@n4, ComponentClass:Actor Component Class Reference=/Script/Engine.SkeletalMeshComponent) -> (ReturnValue:Skeletal Mesh Component Object Reference)
   > Is Valid(InputObject=@n5)
   > Cast To FortPlayerAnimInstance(Object=GetAnimInstance(self=@n5))
   Branch(execute=@n2.then_1, Condition=Get play anim() -> (play anim:bool))
   n6 = GetGameTimeInSeconds()
   n7 = Get WhenStartPlay() -> (WhenStartPlay:double)
   > Set Alpha(Alpha:float=GetRuntimeFloatCurveValue!(
     Curve:"Runtime Float Curve Structure (by ref)"=Get Curve() -> (Curve:Runtime Float Curve Structure)
     InTime:float=float - float(A=@n6, B=@n7)
     InDefaultValue:float=0
    ) -> (ReturnValue:float)) -> (Output_Get:float)
   n8 = GetRuntimeFloatCurveValue(
    Curve=Get Curve_2() -> (Curve_2:Runtime Float Curve Structure)
    InTime=float - float(A=GetGameTimeInSeconds(), B=Get WhenStartPlay())
    InDefaultValue=0
   )
   > Set Alpha_0(Alpha_0:float=@n8) -> (Output_Get:float)
   > Branch(Condition="float >= float"(A=@n8, B=1))
   > Set play anim(play anim:bool=false) -> (Output_Get:bool)
   n9 = float * float(A=GetRealTimeSeconds(), B=300) {is=title K2Node_PromotableOperator}
   Set move baldes(execute=@n2.then_2, move baldes:Rotator=MakeRotator(Pitch=float * float(A=@n9, B=6))) -> (Output_Get:Rotator)
   Event OnWeaponChanged()
   > Set WhenStartPlay(WhenStartPlay:double=GetGameTimeInSeconds()) -> (Output_Get:double)
   > Set play anim(play anim=true)
   TryGetPawnOwner()
   "float <= float"(A=@n8) -> (ReturnValue=~)
  ]}
]
variables [
 Alpha {type=number, default=1}
 Alpha_0 {type=number}
 animation plus delay {type=boolean}
 As Fort Player Anim Instance {type=FortPlayerAnimInstance, default=null}
 Blades on {type=boolean}
 current_velocity {type=number}
 Curve {
  type=RuntimeFloatCurve
  default {
   externalCurve=null
   editorCurveData {
    defaultValue=3.4028234663852886e+38
    postInfinityExtrap=RCCE_Constant
    preInfinityExtrap=RCCE_Constant
    keys [
     {
      arriveTangent=0
      arriveTangentWeight=0
      interpMode=RCIM_Linear
      leaveTangent=0
      leaveTangentWeight=0
      tangentMode=RCTM_Auto
      tangentWeightMode=RCTWM_WeightedNone
      time=0
      value=1
     }
     {
      arriveTangent=-0.5974
      arriveTangentWeight=0
      interpMode=RCIM_Cubic
      leaveTangent=-0.5974
      leaveTangentWeight=0
      tangentMode=RCTM_User
      tangentWeightMode=RCTWM_WeightedNone
      time=0.2
      value=0.9952
     }
     {
      arriveTangent=-7.806
      arriveTangentWeight=0
      interpMode=RCIM_Cubic
      leaveTangent=-7.806
      leaveTangentWeight=0
      tangentMode=RCTM_User
      tangentWeightMode=RCTWM_WeightedNone
      time=0.5
      value=0
     }
    ]
   }
  }
 }
 Curve_2 {
  type=RuntimeFloatCurve
  default {
   externalCurve=null
   editorCurveData {
    defaultValue=3.4028234663852886e+38
    postInfinityExtrap=RCCE_Constant
    preInfinityExtrap=RCCE_Constant
    keys [
     {
      arriveTangent=0
      arriveTangentWeight=0
      interpMode=RCIM_Linear
      leaveTangent=0
      leaveTangentWeight=0
      tangentMode=RCTM_Auto
      tangentWeightMode=RCTWM_WeightedNone
      time=0
      value=0
     }
     {
      arriveTangent=0
      arriveTangentWeight=0
      interpMode=RCIM_Linear
      leaveTangent=0
      leaveTangentWeight=0
      tangentMode=RCTM_Auto
      tangentWeightMode=RCTWM_WeightedNone
      time=0.4866
      value=0.3244
     }
     {
      arriveTangent=2.7401
      arriveTangentWeight=0
      interpMode=RCIM_Cubic
      leaveTangent=-23.5274
      leaveTangentWeight=0
      tangentMode=RCTM_User
      tangentWeightMode=RCTWM_WeightedNone
      time=1.5
      value=1
     }
    ]
   }
  }
 }
 Delta Time X {type=number}
 InterpSpeed {type=number}
 is Animation player melee combo {type=boolean}
 Melee Anim Playing {type=boolean}
 move baldes {type=Rotator}
 Move Blades {type=boolean, default=true}
 play anim {type=boolean}
 RotationSpeed {type=number}
 Rotator {type=Rotator}
 WhenStartPlay {type=number}
]
